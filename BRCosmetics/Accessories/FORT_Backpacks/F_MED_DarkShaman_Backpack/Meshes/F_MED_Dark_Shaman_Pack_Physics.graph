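# Physics asset for the Dark Shaman backpack cosmetic, giving its doll-like body, limbs, left eye and head fabric physics bodies.
# Each bone uses a single capsule, constrained in a tree from backpack_root.
# Body, eye and fabric base attach to the root, arms and legs to the body, and three fabric tips to the fabric base.
# Body, limb and eye limits are tight, while the fabric swings much more freely.
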PhysicsAsset /BRCosmetics/Accessories/FORT_Backpacks/F_MED_DarkShaman_Backpack/Meshes/F_MED_Dark_Shaman_Pack_Physics
bodies {
 backpack_root [{shapeType=Capsule, length=0.0001, radius=6.9508, center=<2.2242 -5.8277 0.8326>, rotation=<rot 0 -76.655 -91.0984>}]
 dyn_arm_left [{shapeType=Capsule, length=0.794, radius=2.0434, center=<-4.1671 -2.1781 -7.5806>, rotation=<rot 3.1902 -89.9995 21.6734>}]
 dyn_arm_right [{shapeType=Capsule, length=1.4563, radius=2.2153, center=<5.1989 -0.8221 -9.8577>, rotation=<rot 1.9365 -89.9998 -7.7242>}]
 dyn_body [{shapeType=Capsule, length=4.5577, radius=4.2732, center=<-0.0174 -0.7478 -6.7619>, rotation=<rot 5.4302 -89.9989 0.2486>}]
 dyn_eye_left [{shapeType=Capsule, length=1.01, radius=2.5958, center=<1.7103 -2.5146 -1.5168>, rotation=<rot -13.8345 -89.9982 17.4105>}]
 dyn_head_fabric_01_top [{shapeType=Capsule, length=0.5326, radius=1.3411, center=<-0.8872 -0.1279 0.498>, rotation=<rot 3.0222 -80.3674 -75.0352>}]
 dyn_head_fabric_02_top [{shapeType=Capsule, length=0.6485, radius=1.1954, center=<0.241 -1.1219 -0.1643>, rotation=<rot -3.0935 -167.2827 -58.1362>}]
 dyn_head_fabric_03_top [{shapeType=Capsule, length=0.5491, radius=0.8999, center=<0.645 0.3276 0.3588>, rotation=<rot 10.8067 119.6235 -78.4388>}]
 dyn_leg_left [{shapeType=Capsule, length=0.3614, radius=2.5271, center=<-0.8526 0.2257 -8.373>, rotation=<rot -2.3911 -74.2622 2.9474>}]
 dyn_leg_right [{shapeType=Capsule, length=0.4092, radius=2.4443, center=<-0.0248 0.2354 -7.8957>, rotation=<rot -1.9666 -122.7661 -4.1105>}]
 head_fabric_base [{shapeType=Capsule, length=0.75, radius=2.468, center=<0 0 -0.4015>, rotation=<rot 0 -90 0>}]
}
constraints [
 {bone1Name=head_fabric_base, bone2Name=backpack_root, swing1LimitDegrees=45, swing2LimitDegrees=45, twistLimitDegrees=1}
 {bone1Name=dyn_arm_right, bone2Name=dyn_body, swing1LimitDegrees=5, swing2LimitDegrees=3, twistLimitDegrees=1}
 {bone1Name=dyn_arm_left, bone2Name=dyn_body, swing1LimitDegrees=3, swing2LimitDegrees=3, twistLimitDegrees=1}
 {bone1Name=dyn_leg_left, bone2Name=dyn_body, swing1LimitDegrees=3, swing2LimitDegrees=3, twistLimitDegrees=1}
 {bone1Name=dyn_leg_right, bone2Name=dyn_body, swing1LimitDegrees=10, swing2LimitDegrees=10, twistLimitDegrees=0.5}
 {bone1Name=dyn_body, bone2Name=backpack_root, swing1LimitDegrees=7, swing2LimitDegrees=7, twistLimitDegrees=1}
 {bone1Name=dyn_head_fabric_01_top, bone2Name=head_fabric_base, swing1LimitDegrees=60, swing2LimitDegrees=60, twistLimitDegrees=10}
 {bone1Name=dyn_head_fabric_02_top, bone2Name=head_fabric_base, swing1LimitDegrees=70, swing2LimitDegrees=70, twistLimitDegrees=10}
 {bone1Name=dyn_head_fabric_03_top, bone2Name=head_fabric_base, swing1LimitDegrees=50, swing2LimitDegrees=50, twistLimitDegrees=10}
 {bone1Name=dyn_eye_left, bone2Name=backpack_root, swing1LimitDegrees=10, swing2LimitDegrees=5, twistLimitDegrees=1}
]
